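; Blueprint for the Mask Maker portal lighting in Mask Fortress, controlling a point light and a spotlight.
; On BeginPlay it binds SettingsChange to the game user settings UI update dispatcher.
; SettingsChange runs a VFX shadow quality check to toggle point light shadow casting, enabling it on the Epic branch.
; A DisablePortalLighting event sets both light intensities to zero; overlap, tick and construction script are empty.
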(event Custom|SettingsChange
  (Utilities|ShadowQualityCheckVFX
    (:Medium
      _)
    (:High
      _)
    (:Epic
      (Rendering|Components|Light|SetCastShadows (Variables|Default|GetBR_Pointlight_Component) true))
    (:Cinematic
      _))
  (Rendering|Components|Light|SetCastShadows (Variables|Default|GetBR_Pointlight_Component) false))

(event EventBeginPlay
  (bind _returnvalue (Settings|GetGameUserSettings))
  (EventDispatchers|BindEventtoOnGameUserSettingsUINeedsUpdate _returnvalue (delegate Custom|SettingsChange)))

(event Custom|DisablePortalLighting
  (Rendering|Components|Light|SetIntensity (Variables|Default|GetBR_Pointlight_Component) 0.0)
  (Rendering|Components|Light|SetIntensity (Variables|Default|GetBR_Spotlight_Component) 0.0))

(event Collision|EventActorBeginOverlap
  (params (OtherActor :type Actor%20Object%20Reference)))

(event EventTick
  (params (DeltaSeconds :type Float%20%28single-precision%29)))

(fn ConstructionScript)
