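# Light function material for the Mask Fortress core in the Chapter 6 map, modulating light with swirling noise and a fire-like flicker.
# A rotating swirly noise texture sets spatial variation above a NoiseMin floor.
# A looping curve-atlas flicker, paced by Timing, sets the pulse above PulseMin.
# Their product is scaled by Portal_Intensity from the FortniteMaterialParameters collection, allowing shared external control of brightness.
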Material /BRMapCh6/FX/MaskFortress/BP/LF_Lightfunction_MaskFortress_Core
graph {
 properties {bAutomaticallySetUsageInEditor=false, materialDomain=MD_LightFunction, usedShadingModels=Default Lit}
 nodes [
  n1 = ScalarParameter() {defaultValue=0.7, parameterName=NoiseMin}
  n2 = TextureSample(UVs=Rotator(Coordinate=TextureCoordinate() {uTiling=0.5, vTiling=0.5})) {texture=/Game/Weapons/WeaponSkins/Textures/T_Swirly_Noise_highContrast}
  n3 = ScalarParameter() {defaultValue=0.25, parameterName=PulseMin}
  n4 = Time()
  n5 = ScalarParameter() {defaultValue=0.2, parameterName=Timing}
  n6 = CurveAtlasRowParameter(CurveTime=Frac(None=Multiply(A=@n4, B=@n5))) {
   atlas=/Game/Environments/Helios/Props/Brimstone/Brimstone_Brazier_A/Blueprints/CurveAtlas_FN_LightFunctions
   curve=/Game/Blueprints/Lighting/LightFunctions/Curve_FireFlickers
   parameterName=Flickercurve
  }
  n7 = CollectionParameter() {collection=&FortniteMaterialParameters, parameterName=Portal_Intensity}
  Multiply(A=Multiply(A=LinearInterpolate(A=@n1, Alpha=@n2.R) {constA=0.7, constB=1.33}, B=LinearInterpolate(A=@n3, Alpha=@n6.R)), B=@n7)
 ]
}
