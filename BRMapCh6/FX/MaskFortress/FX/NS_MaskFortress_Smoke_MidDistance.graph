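# Niagara system for mid-distance smoke and sparkle ambience at the Mask Fortress, using GPU emitters with distance-scaled spawning and culling.
# SparklesParent scatters particles onto the depth buffer, culled by view mask, and seeds the Sparkles sprites and Ground_Lines meshes.
# Sky adds mesh particles in a large sphere, while Ground_MidDist, a landscape-sampling ring, is disabled.
# Negative point forces move particles, and radial region masks vary color, with per-particle material parameters.
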NiagaraSystem /BRMapCh6/FX/MaskFortress/FX/NS_MaskFortress_Smoke_MidDistance
dependencies {
 usedDataInterfaces [
  &NiagaraDataInterfaceCurve
  &NiagaraDataInterfaceMeshRendererInfo
  &NiagaraDataInterfaceGBuffer
  &NiagaraDataInterfaceCollisionQuery
  &NiagaraDataInterfaceColorCurve
  &NiagaraDataInterfaceParticleRead
 ]
 usedDynamicInputs []
 usedModules [
  &SystemState
  &self:SystemSpawnScript.SystemScriptSource.SystemScriptGraph.NiagaraNodeAssignment_0.SetVariables_#1
  &self:SystemSpawnScript.SystemScriptSource.SystemScriptGraph.NiagaraNodeAssignment_1.SetVariables_#2
  &self:SystemSpawnScript.SystemScriptSource.SystemScriptGraph.NiagaraNodeAssignment_2.SetVariables_#3
  &self:SystemSpawnScript.SystemScriptSource.SystemScriptGraph.NiagaraNodeAssignment_3.SetVariables_#4
 ]
}
emitters {
 Ground_Lines {
  simTarget=GPUComputeSim
  emitterUpdateScript [
   &EmitterState(
    Enable Distance Culling=true
    Enable Visibility Culling=true
    MaxDistance=50000
    Min Distance Response=::SleepAndLetParticlesFinish
    Module.bApplySpawnCountScaleByDistance=true
    Module.bUseMaxDistance=true
    Module.bUseMinDistance=true
    Reset Age On Awaken=true
    Scalability Mode=::NewEnumerator1
    Scale Spawn Count=true
    Spawn Count Scale By Distance=DataInterface() {propertyValues {Curve {keys [<at 0.1 0>,<at 0.2 1>,<at 0.8 1>,<at 0.9 0>]}}}
   )
   &SpawnParticlesFromOtherEmitter(
    Attribute Reader=DataInterface() {propertyValues {EmitterBinding {bindingMode=Other, emitterName=SparklesParent}}}
    Calculate Spawn Rate Per Particle=false
    Spawn Rate=500
    Use Maximum Age=false
    Use Minimum Age=false
   )
  ]
  particleSpawnScript [
   &InitializeParticle(
    Lifetime=&RandomRangeFloat(Minimum=3, Maximum=4)
    Mesh Scale Mode=::NewEnumerator1
    Mesh Uniform Scale Max=1.5
    UsePositionOffset=false
    Write Material Random=true
   )
   &ShapeLocation(MaskSpawnGroup=false, Override Seed=false, Sphere Radius=1, UseOwnerScale=true)
   &InitialMeshOrientation(
    Mesh Axis To Orient=<0 0 1>
    Mesh Orientation Mode=::NewEnumerator2
    Orientation Coordinate Space=::World
    Rotation=<0 0 0.75>
    Use Rotation Vector=true
   )
   &AddVelocity(Module.Use Falloff Distance=false, Override Seed=false, Velocity Mode=::NewEnumerator1, Velocity Speed=0) {module=AddVelocity001}
   &SampleParticlesFromOtherEmitter(Applied Velocity Scale=0, Velocity Sampling=::NewEnumerator0)
  ]
  particleUpdateScript [
   &CurlNoiseForce(Noise Frequency=0.5, Noise Strength=500, Pan Noise=false) {enabled=false}
   &ParticleState()
   &RegionMask(Region Falloff=1, Region Radius=15000, UseOffset=false, UseRotation=false)
   &RegionMask(Inner Float=0.25, Outer Float=0.1, Region Falloff=1, Region Radius=8000, UseOffset=false, UseRotation=false) {module=RegionMask001}
   &ScaleMeshSize(Scale Factor=&Multiply_VectorByFloat(
     Vector=<1 1 1.5>
     Float=&FloatFromCurve(
      FloatCurve=DataInterface() {propertyValues {Curve {
         defaultValue=1
         keys [
          {arriveTangent=0.0927, interpMode=RCIM_Cubic, leaveTangent=0.0927, tangentMode=RCTM_User, time=0, value=0}
          {tangentMode=RCTM_User, time=1, value=0.5}
         ]
        }}}
      Scale Curve=&Multiply_Float(A=1, B=$Particles.RegionMask.Float)
     )
    ))
   &PointForce(
    Force Falloff Distance=50000
    Force Origin=&AddVectorToPosition(Position=/Niagara/DynamicInputs/Helpers/SimulationPosition(), Vector=<0 0 500>)
    Force Strength=-250
    Module.Use Falloff=true
    Module.UseLinearForce=false
   )
   /Niagara/Modules/Update/Camera/ViewMask(Border Percent Max=50, Border Percent Min=15)
   &ScaleColor(
    Scale Alpha=&FloatFromCurve(
     FloatCurve=DataInterface() {propertyValues {Curve {keys [
         {interpMode=RCIM_Cubic, tangentMode=RCTM_User, time=0, value=0}
         {interpMode=RCIM_Cubic, tangentMode=RCTM_User, time=0.5, value=0.95}
         <at 1 0>
        ]}}}
     Scale Curve=&Multiply_Float(
      A=0.05
      B=&FloatFromCurve(
       FloatCurve=DataInterface() {propertyValues {Curve {
          defaultValue=1
          keys [
           {interpMode=RCIM_Cubic, tangentMode=RCTM_User, time=0, value=0.5}
           {interpMode=RCIM_Cubic, tangentMode=RCTM_User, time=0.5, value=1}
           {tangentMode=RCTM_User, time=1, value=1}
          ]
         }}}
       CurveIndex=$Particles.RegionMask.Float
      )
     )
    )
    Scale RGB=&Multiply_VectorByFloat(
     Vector=&MakeVectorFromLinearColorRGB(Color=&Multiply_LinearColorByFloat(
       Linear Color=&ColorFromCurve(
        ColorCurve=DataInterface() {propertyValues {
          AlphaCurve {keys [<at 0 1>]}
          BlueCurve {keys [<at 0 0.0345 RCIM_Cubic>,<at 0.6237 0.0528>]}
          GreenCurve {keys [<at 0 0.0706 RCIM_Cubic>,<at 0.6237 0.1309>]}
          RedCurve {keys [<at 0 0.3021 RCIM_Cubic>,<at 0.6237 0.3854>]}
         }}
        CurveIndex=$Particles.RegionMask.Float
       )
       Float=&RandomRangeFloat(Minimum=1, Maximum=2)
      ))
     Float=&FloatFromCurve(FloatCurve=DataInterface() {propertyValues {Curve {keys [{interpMode=RCIM_Cubic, tangentMode=RCTM_User, time=0, value=0}, {interpMode=RCIM_Cubic, tangentMode=RCTM_User, time=0.6, value=1}, <at 1 0>]}}})
    )
    ScaleA=true
    ScaleRGB=true
   )
   &DynamicMaterialParameters(
    Index 0 Param 1=&Multiply_Float(
     A=&RandomRangeFloat(Minimum=10, Maximum=25)
     B=&FloatFromCurve(FloatCurve=DataInterface() {propertyValues {Curve {defaultValue=1, keys [{arriveTangent=9.0725, time=0, value=0}, <at 0.2 0>, <at 0.9 2>]}}})
    )
    Index 0 Param 2=&Multiply_Float(
     A=&Multiply_Float(
      A=&RandomRangeFloat(Minimum=2, Maximum=3)
      B=&Lerp_Float(A=15, B=0.2, Alpha=$Particles.RegionMask.Float)
     )
     B=&FloatFromCurve(
      FloatCurve=DataInterface() {propertyValues {Curve {defaultValue=1, keys [{arriveTangent=9.0725, time=0, value=0}, <at 0.2 0>, <at 0.9 2>]}}}
      Scale Curve=1.01
     )
    )
    Index 0 Param 3=&FloatFromCurve(
     FloatCurve=DataInterface() {propertyValues {Curve {defaultValue=1, keys [{arriveTangent=9.0725, time=0, value=0.5}, <at 0.9 0.5>]}}}
     Scale Curve=2
    )
    Index 0 Param 4=&RandomRangeFloat(Minimum=0.25, Maximum=1)
    Index 1 Param 1=&RandomRangeFloat(Minimum=1, Maximum=2)
    Index 1 Param 2=&RandomRangeFloat(Minimum=500, Maximum=1500)
    Index 1 Param 3=&RandomRangeFloat(Minimum=500, Maximum=2500)
    Index 1 Param 4=&RandomRangeFloat(Minimum=0.1, Maximum=0.15)
    Index 2 Param 1=&RandomRangeFloat(Minimum=0, Maximum=1)
    Index 2 Param 2=&RandomRangeFloat(Minimum=0.2, Maximum=1)
    Index 2 Param 3=&RandomRangeFloat(Minimum=0.25, Maximum=1)
    Index 3 Param 1=&Multiply_Float(A=100, B=$Particles.RegionMask001.Float)
    Index 3 Param 2=&Multiply_Float(A=1, B=$Particles.RegionMask001.Float)
    Index 3 Param 3=1
    Module.Index0Float0Write=true
    Module.Index0Float1Write=true
    Module.Index0Float2Write=true
    Module.Index0Float3Write=true
    Module.Index1Float0Write=true
    Module.Index1Float1Write=true
    Module.Index1Float2Write=true
    Module.Index1Float3Write=true
    Module.Index2Float0Write=true
    Module.Index2Float1Write=true
    Module.Index2Float2Write=true
    Module.Index2Float3Write=true
    Module.Index3Float0Write=true
    Module.Index3Float1Write=true
    Module.Index3Float2Write=true
    Module.Index3Float3Write=true
    Write Parameter Index 1=::NewEnumerator1
    Write Parameter Index 2=::NewEnumerator1
    Write Parameter Index 3=::NewEnumerator1
   )
   &SolveForcesAndVelocity(Clamp Velocity=true, Limit Acceleration=false, Speed Limit=150)
   &CalculateAccurateVelocity()
   &CameraOffset(Camera Offset Amount=&RandomRangeFloat(Minimum=15, Maximum=150))
  ]
  rendererClasses [&NiagaraMeshRendererProperties]
  renderers [{rendererClass=&NiagaraMeshRendererProperties, rendererIndex=0}]
 }
 Ground_MidDist {
  bEnabled=false
  simTarget=GPUComputeSim
  emitterUpdateScript [
   &EmitterState(
    Enable Distance Culling=true
    Enable Visibility Culling=true
    MaxDistance=50000
    Min Distance Response=::SleepAndLetParticlesFinish
    Module.bApplySpawnCountScaleByDistance=true
    Module.bUseMaxDistance=true
    Module.bUseMinDistance=true
    Reset Age On Awaken=true
    Scalability Mode=::NewEnumerator1
    Scale Spawn Count=true
    Spawn Count Scale By Distance=DataInterface() {propertyValues {Curve {keys [<at 0.1 0>,<at 0.2 1>,<at 0.8 1>,<at 0.9 0>]}}}
   )
   &SpawnRate(SpawnRate=250, Use Spawn Probability=false)
  ]
  particleSpawnScript [
   &InitializeParticle(
    Color=<rgba 1 0.3208 0.9829 1>
    Lifetime=&RandomRangeFloat(Minimum=3, Maximum=4)
    Mesh Scale Mode=::NewEnumerator1
    Mesh Uniform Scale Max=1
    UsePositionOffset=false
    Write Material Random=true
   )
   &ShapeLocation(Disc Coverage=0.869, MaskSpawnGroup=false, Override Seed=false, Ring Radius=10000, Shape Primitive=::NewEnumerator4, UseOwnerScale=true)
   &self:Sample_Landscape() {module=Sample_Landscape}
   &InitialMeshOrientation(Mesh Axis To Orient=<0 0 1>, Mesh Orientation Mode=::NewEnumerator2, Rotation=<0 0 0.75>, Use Rotation Vector=true)
   &AddVelocity(Module.Use Falloff Distance=false, Override Seed=false, Velocity Mode=::NewEnumerator1, Velocity Speed=-1000) {module=AddVelocity001}
  ]
  particleUpdateScript [
   &CurlNoiseForce(Noise Frequency=0.5, Noise Strength=500, Pan Noise=false) {enabled=false}
   &ParticleState()
   &ScaleMeshSize(Scale Factor=&Multiply_VectorByFloat(
     Vector=<1 1 1.5>
     Float=&FloatFromCurve(FloatCurve=DataInterface() {propertyValues {Curve {
         defaultValue=1
         keys [
          {arriveTangent=0.0927, interpMode=RCIM_Cubic, leaveTangent=0.0927, tangentMode=RCTM_User, time=0, value=0}
          {tangentMode=RCTM_User, time=1, value=1}
         ]
        }}})
    ))
   &RegionMask(Region Falloff=1, Region Radius=12500, UseOffset=false, UseRotation=false)
   &RegionMask(Region Falloff=0, Region Radius=15000, UseOffset=false, UseRotation=false) {enabled=false, module=RegionMask001}
   &PointForce(Force Falloff Distance=50000, Force Strength=-2500, Module.Use Falloff=true, Module.UseLinearForce=false)
   &self:Sample_Landscape() {module=Sample_Landscape001}
   &ScaleColor(
    Scale Alpha=&FloatFromCurve(
     FloatCurve=DataInterface() {propertyValues {Curve {keys [
         {interpMode=RCIM_Cubic, tangentMode=RCTM_User, time=0, value=0}
         {arriveTangent=2, interpMode=RCIM_Cubic, leaveTangent=2, time=0.1, value=0.65}
         {interpMode=RCIM_Cubic, tangentMode=RCTM_User, time=0.5, value=1}
         <at 1 0>
        ]}}}
     Scale Curve=&Multiply_Float(A=15, B=$System.Main_Alpha)
    )
    ScaleA=true
    ScaleRGB=true
   )
   &ScaleColor(
    Scale Alpha=&Multiply_Float(
     A=&FloatFromCurve(
      FloatCurve=DataInterface() {propertyValues {Curve {keys [
          {interpMode=RCIM_Cubic, tangentMode=RCTM_User, time=0, value=0}
          {arriveTangent=4.75, interpMode=RCIM_Cubic, leaveTangent=4.75, time=0.1, value=0.65}
          {interpMode=RCIM_Cubic, tangentMode=RCTM_User, time=0.2, value=0.95}
          <at 1 0>
         ]}}}
      Scale Curve=5
     )
     B=/Niagara/DynamicInputs/Scalability/ScalabilityDistanceBasedFloat(
      Curve:NiagaraDataInterfaceCurve=DataInterface() {propertyValues {Curve {keys [
          {interpMode=RCIM_Cubic, tangentMode=RCTM_User, time=0, value=0}
          {interpMode=RCIM_Cubic, tangentMode=RCTM_User, time=0.1, value=1}
          <at 0.5 1>
          {interpMode=RCIM_Cubic, tangentMode=RCTM_User, time=0.9, value=1}
          {interpMode=RCIM_Cubic, tangentMode=RCTM_User, time=1, value=0}
         ]}}}
      CurveIndex:NiagaraFloat=$Emitter.Scalability.DistanceFraction
      Bias:NiagaraFloat=0
      Scale:NiagaraFloat=1
     )
    )
    Scale RGB=<5 5 5>
    ScaleA=true
    ScaleRGB=true
   ) {module=ScaleColor001}
   /SetVariables(Particles.Position=&AddVectorToPosition(Position=$Particles.Position, Vector=$System.TerrainYp)) {isSetParameters=true}
   &DynamicMaterialParameters(
    Index 0 Param 1=&Multiply_Float(
     A=&RandomRangeFloat(Minimum=50, Maximum=100)
     B=&NormalizeDistanceRange(
      Start Position:NiagaraPosition=$Particles.Position
      End Position:NiagaraPosition=$System.CameraPostion
      Distance:NiagaraFloat=5000
      Invert Normalized Range:NiagaraBool=false
     )
    )
    Index 0 Param 2=&Multiply_Float(
     A=&RandomRangeFloat(Minimum=2, Maximum=5)
     B=&NormalizeDistanceRange(
      Start Position:NiagaraPosition=$Particles.Position
      End Position:NiagaraPosition=$System.CameraPostion
      Distance:NiagaraFloat=5000
      Invert Normalized Range:NiagaraBool=false
     )
    )
    Index 0 Param 3=&FloatFromCurve(
     FloatCurve=DataInterface() {propertyValues {Curve {defaultValue=1, keys [{arriveTangent=9.0725, time=0, value=0.5}, <at 0.2 0.49>, <at 0.9 0.47>]}}}
     Scale Curve=1.01
    )
    Index 0 Param 4=&RandomRangeFloat(Minimum=0.25, Maximum=0.35)
    Index 1 Param 1=&RandomRangeFloat(Minimum=25, Maximum=50)
    Index 1 Param 2=&RandomRangeFloat(Minimum=1500, Maximum=2500)
    Index 1 Param 3=&RandomRangeFloat(Minimum=200, Maximum=500)
    Index 1 Param 4=&RandomRangeFloat(Minimum=1, Maximum=2)
    Index 2 Param 1=&RandomRangeFloat(Minimum=0, Maximum=1)
    Index 2 Param 2=&RandomRangeFloat(Minimum=0.2, Maximum=0.5)
    Index 2 Param 3=&RandomRangeFloat(Minimum=0.25, Maximum=0.4)
    Index 2 Param 4=&FloatFromCurve(
     FloatCurve=DataInterface() {propertyValues {Curve {keys [<at 0 1>,<at 1 1>]}}}
     Scale Curve=&RandomRangeFloat(Minimum=0.9, Maximum=1)
    )
    Index 3 Param 1=1
    Index 3 Param 2=25
    Module.Index0Float0Write=true
    Module.Index0Float1Write=true
    Module.Index0Float2Write=true
    Module.Index0Float3Write=true
    Module.Index1Float0Write=true
    Module.Index1Float1Write=true
    Module.Index1Float2Write=true
    Module.Index1Float3Write=true
    Module.Index2Float0Write=true
    Module.Index2Float1Write=true
    Module.Index2Float2Write=true
    Module.Index2Float3Write=true
    Module.Index3Float0Write=true
    Module.Index3Float1Write=true
    Module.Index3Float2Write=true
    Module.Index3Float3Write=true
    Write Parameter Index 1=::NewEnumerator1
    Write Parameter Index 2=::NewEnumerator1
    Write Parameter Index 3=::NewEnumerator1
   )
   &SolveForcesAndVelocity(Clamp Velocity=true, Limit Acceleration=false, Speed Limit=500)
   &CameraOffset(Camera Offset Amount=1500)
   /SetVariables(Particles.Linear Color=&MultiplyLinearColor(
     Linear Color=$Particles.Color
     Scale Factor=&MakeVector4(
      X=0
      Y=0
      Z=&FloatFromCurve(FloatCurve=DataInterface() {propertyValues {Curve {keys [{arriveTangent=-3.7681, interpMode=RCIM_Cubic, leaveTangent=-3.7681, tangentMode=RCTM_User, time=0, value=1}, <at 0.25 0>]}}})
      W=&FloatFromCurve(FloatCurve=DataInterface() {propertyValues {Curve {keys [
           {interpMode=RCIM_Cubic, tangentMode=RCTM_User, time=0, value=0}
           {interpMode=RCIM_Cubic, tangentMode=RCTM_User, time=0.3, value=0.95}
           <at 0.8 1>
           <at 1 0>
          ]}}})
     )
    )) {isSetParameters=true, module=SetVariables_#5}
   /SetVariables(Particles.Vector:Vector3f=&Multiply_Vector(A=$StackContext.Scale, B=<2 2 1.2>)) {isSetParameters=true, module=SetVariables_#6}
   /SetVariables(Particles.float:NiagaraFloat=&Multiply_Float(A=$Particles.CameraOffset, B=1.1)) {isSetParameters=true, module=SetVariables_#7}
  ]
  rendererClasses [&NiagaraMeshRendererProperties]
  renderers [
   {rendererClass=&NiagaraMeshRendererProperties, rendererIndex=0}
   {rendererClass=&NiagaraMeshRendererProperties, rendererIndex=1}
   {rendererClass=&NiagaraMeshRendererProperties, rendererIndex=2}
  ]
 }
 Sky {
  simTarget=GPUComputeSim
  emitterUpdateScript [
   &EmitterState(
    Enable Distance Culling=true
    Enable Visibility Culling=true
    MaxDistance=50000
    Min Distance Response=::SleepAndLetParticlesFinish
    Module.bApplySpawnCountScaleByDistance=true
    Module.bUseMaxDistance=true
    Module.bUseMinDistance=true
    Reset Age On Awaken=true
    Scalability Mode=::NewEnumerator1
    Scale Spawn Count=true
    Spawn Count Scale By Distance=DataInterface() {propertyValues {Curve {keys [<at 0.1 0>,<at 0.2 1>,<at 0.8 1>,<at 0.9 0>]}}}
   )
   &SpawnRate(SpawnRate=500, Use Spawn Probability=false)
  ]
  particleSpawnScript [
   &InitializeParticle(
    Lifetime=&RandomRangeFloat(Minimum=3, Maximum=4)
    Mesh Scale Mode=::NewEnumerator1
    Sprite Size Mode=::NewEnumerator3
    Uniform Sprite Size=50
    UsePositionOffset=false
    Write Material Random=true
   )
   &ShapeLocation(MaskSpawnGroup=false, Override Seed=false, Sphere Radius=10000, UseOwnerScale=true)
   &InitialMeshOrientation(
    Mesh Axis To Orient=<0 0 1>
    Mesh Orientation Mode=::NewEnumerator2
    Orientation Coordinate Space=::World
    Rotation=<0 0 0.75>
    Use Rotation Vector=true
   )
   &AddVelocity(Module.Use Falloff Distance=false, Override Seed=false, Velocity Mode=::NewEnumerator1, Velocity Speed=-250) {module=AddVelocity001}
  ]
  particleUpdateScript [
   &CurlNoiseForce(Noise Frequency=0.5, Noise Strength=500, Pan Noise=false) {enabled=false}
   &ParticleState()
   &ScaleMeshSize(Scale Factor=&Multiply_VectorByFloat(
     Vector=<1 1 1.5>
     Float=&FloatFromCurve(FloatCurve=DataInterface() {propertyValues {Curve {
         defaultValue=1
         keys [
          {arriveTangent=0.0927, interpMode=RCIM_Cubic, leaveTangent=0.0927, tangentMode=RCTM_User, time=0, value=0}
          {tangentMode=RCTM_User, time=1, value=0.5}
         ]
        }}})
    ))
   &RegionMask(Region Falloff=1, Region Radius=15000, UseOffset=false, UseRotation=false)
   &RegionMask(Inner Float=0, Region Falloff=1, Region Radius=250, UseOffset=false, UseRotation=false) {module=RegionMask001}
   &PointForce(
    Force Falloff Distance=15000
    Force Origin=&AddVectorToPosition(Position=&SimulationPosition(), Vector=<0 0 500>)
    Force Strength=-1500
    Module.Use Falloff=true
    Module.UseLinearForce=false
   )
   &ScaleColor(
    Scale Alpha=&FloatFromCurve(
     FloatCurve=DataInterface() {propertyValues {Curve {keys [
         {interpMode=RCIM_Cubic, tangentMode=RCTM_User, time=0, value=0}
         {interpMode=RCIM_Cubic, tangentMode=RCTM_User, time=0.5, value=0.95}
         <at 1 0>
        ]}}}
     Scale Curve=0.05
    )
    Scale RGB=&Multiply_VectorByFloat(
     Vector=&MakeVectorFromLinearColorRGB(Color=&Multiply_LinearColorByFloat(
       Linear Color=&ColorFromCurve(
        ColorCurve=DataInterface() {propertyValues {
          AlphaCurve {keys [<at 0 1>]}
          BlueCurve {keys [<at 0 0.0168 RCIM_Cubic>,<at 0.4441 0.0585>]}
          GreenCurve {keys [<at 0 0.0343 RCIM_Cubic>,<at 0.4441 0.0989>]}
          RedCurve {keys [<at 0 0.147 RCIM_Cubic>,<at 0.4441 0.6042>]}
         }}
        CurveIndex=$Particles.RegionMask.Float
       )
       Float=&RandomRangeFloat(Minimum=1, Maximum=5)
      ))
     Float=&FloatFromCurve(FloatCurve=DataInterface() {propertyValues {Curve {keys [{interpMode=RCIM_Cubic, tangentMode=RCTM_User, time=0, value=0}, {interpMode=RCIM_Cubic, tangentMode=RCTM_User, time=0.6, value=1}, <at 1 0>]}}})
    )
    ScaleA=true
    ScaleRGB=true
   )
   &ScaleColor(
    Scale Alpha=&Multiply_Float(
     A=&FloatFromCurve(FloatCurve=DataInterface() {propertyValues {Curve {keys [
          {interpMode=RCIM_Cubic, tangentMode=RCTM_User, time=0, value=0}
          {arriveTangent=4.75, interpMode=RCIM_Cubic, leaveTangent=4.75, time=0.1, value=0.65}
          {interpMode=RCIM_Cubic, tangentMode=RCTM_User, time=0.2, value=0.95}
          <at 1 0>
         ]}}})
     B=&ScalabilityDistanceBasedFloat(
      Curve:NiagaraDataInterfaceCurve=DataInterface() {propertyValues {Curve {keys [
          {interpMode=RCIM_Cubic, tangentMode=RCTM_User, time=0, value=0}
          {interpMode=RCIM_Cubic, tangentMode=RCTM_User, time=0.1, value=1}
          <at 0.5 1>
          {interpMode=RCIM_Cubic, tangentMode=RCTM_User, time=0.9, value=1}
          {interpMode=RCIM_Cubic, tangentMode=RCTM_User, time=1, value=0}
         ]}}}
      CurveIndex:NiagaraFloat=$Emitter.Scalability.DistanceFraction
      Bias:NiagaraFloat=0
      Scale:NiagaraFloat=1
     )
    )
    Scale RGB=&Lerp_Vector(
     A=&MakeVectorFromLinearColorRGB(Color=$Particles.Color)
     B=&MakeVectorFromLinearColorRGB(Color=<rgba 1500 1.3141 360.9798 1>)
     Alpha=$Particles.RegionMask001.Float
    )
    ScaleA=true
    ScaleRGB=true
   ) {enabled=false, module=ScaleColor001}
   &DynamicMaterialParameters(
    Index 0 Param 1=&Multiply_Float(
     A=&RandomRangeFloat(Minimum=2, Maximum=5)
     B=&FloatFromCurve(FloatCurve=DataInterface() {propertyValues {Curve {defaultValue=1, keys [{arriveTangent=9.0725, time=0, value=0}, <at 0.2 0>, <at 0.9 2>]}}})
    )
    Index 0 Param 2=&Multiply_Float(
     A=&RandomRangeFloat(Minimum=15, Maximum=25)
     B=&FloatFromCurve(
      FloatCurve=DataInterface() {propertyValues {Curve {defaultValue=1, keys [{arriveTangent=9.0725, time=0, value=0}, <at 0.2 0>, <at 0.9 2>]}}}
      Scale Curve=1.01
     )
    )
    Index 0 Param 3=&FloatFromCurve(
     FloatCurve=DataInterface() {propertyValues {Curve {defaultValue=1, keys [{arriveTangent=9.0725, time=0, value=0.5}, <at 0.9 0.5>]}}}
     Scale Curve=1.1
    )
    Index 0 Param 4=&RandomRangeFloat(Minimum=0.25, Maximum=0.35)
    Index 1 Param 1=&RandomRangeFloat(Minimum=1, Maximum=2)
    Index 1 Param 2=&RandomRangeFloat(Minimum=800, Maximum=2500)
    Index 1 Param 3=&RandomRangeFloat(Minimum=250, Maximum=2500)
    Index 1 Param 4=&RandomRangeFloat(Minimum=0.1, Maximum=0.25)
    Index 2 Param 1=&RandomRangeFloat(Minimum=0, Maximum=1)
    Index 2 Param 2=&RandomRangeFloat(Minimum=0.2, Maximum=0.5)
    Index 2 Param 3=&RandomRangeFloat(Minimum=0.25, Maximum=0.5)
    Index 2 Param 4=&FloatFromCurve(
     FloatCurve=DataInterface() {propertyValues {Curve {keys [<at 0 1>,<at 1 1>]}}}
     Scale Curve=&RandomRangeFloat(Minimum=0, Maximum=0)
    )
    Index 3 Param 1=1
    Index 3 Param 2=5
    Module.Index0Float0Write=true
    Module.Index0Float1Write=true
    Module.Index0Float2Write=true
    Module.Index0Float3Write=true
    Module.Index1Float0Write=true
    Module.Index1Float1Write=true
    Module.Index1Float2Write=true
    Module.Index1Float3Write=true
    Module.Index2Float0Write=true
    Module.Index2Float1Write=true
    Module.Index2Float2Write=true
    Module.Index2Float3Write=true
    Module.Index3Float0Write=true
    Module.Index3Float1Write=true
    Module.Index3Float2Write=true
    Module.Index3Float3Write=true
    Write Parameter Index 1=::NewEnumerator1
    Write Parameter Index 2=::NewEnumerator1
    Write Parameter Index 3=::NewEnumerator1
   )
   &SolveForcesAndVelocity(Clamp Velocity=true, Limit Acceleration=false, Speed Limit=250)
   &CameraOffset(Camera Offset Amount=1500)
   /SetVariables(Particles.Linear Color=&MultiplyLinearColor(
     Linear Color=$Particles.Color
     Scale Factor=&MakeVector4(
      X=0
      Y=0
      Z=&FloatFromCurve(FloatCurve=DataInterface() {propertyValues {Curve {keys [
           {arriveTangent=-3.7681, interpMode=RCIM_Cubic, leaveTangent=-3.7681, tangentMode=RCTM_User, time=0, value=0}
           <at 0.1 1 RCIM_Cubic>
           <at 0.25 0>
          ]}}})
      W=&FloatFromCurve(
       FloatCurve=DataInterface() {propertyValues {Curve {keys [
           {interpMode=RCIM_Cubic, tangentMode=RCTM_User, time=0, value=0}
           {interpMode=RCIM_Cubic, tangentMode=RCTM_User, time=0.3, value=0.95}
           <at 0.8 1>
           <at 1 0>
          ]}}}
       Scale Curve=0.5
      )
     )
    )) {isSetParameters=true, module=SetVariables_#5}
   /SetVariables(Particles.Vector:Vector3f=&Multiply_Vector(A=$StackContext.Scale, B=<1 1 1.2>)) {isSetParameters=true, module=SetVariables_#6}
   /SetVariables(Particles.float:NiagaraFloat=&Multiply_Float(A=$Particles.CameraOffset, B=1.1)) {isSetParameters=true, module=SetVariables_#7}
  ]
  rendererClasses [&NiagaraMeshRendererProperties]
  renderers [
   {rendererClass=&NiagaraMeshRendererProperties, rendererIndex=0}
   {rendererClass=&NiagaraMeshRendererProperties, rendererIndex=1}
   {rendererClass=&NiagaraMeshRendererProperties, rendererIndex=2}
  ]
 }
 Sparkles {
  simTarget=GPUComputeSim
  emitterUpdateScript [
   &EmitterState(
    Enable Distance Culling=true
    MaxDistance=25000
    Module.bApplySpawnCountScaleByDistance=true
    Module.bUseMaxDistance=true
    Module.bUseMinDistance=false
    Scalability Mode=::NewEnumerator1
    Scale Spawn Count=true
    Spawn Count Scale By Distance=DataInterface() {propertyValues {Curve {keys [<at 0 0>,<at 0.1 0>,<at 0.5 1>,<at 0.75 1>,<at 1 1>]}}}
   )
   &SpawnParticlesFromOtherEmitter(
    Attribute Reader=DataInterface() {propertyValues {EmitterBinding {bindingMode=Other, emitterName=SparklesParent}}}
    Spawn Rate=250
    Spawn Rate Per Particle Cap=50
    Use Maximum Age=false
    Use Minimum Age=false
   )
  ]
  particleSpawnScript [
   &InitializeParticle(
    Lifetime Max=3
    Lifetime Min=2
    Lifetime Mode=::NewEnumerator1
    Sprite Size Mode=::NewEnumerator1
    UsePositionOffset=false
    Write Material Random=true
   ) {module=InitializeParticle001}
   &CameraOffset(Camera Offset Amount=50)
   /SetVariables(Particles.InitialPos:NiagaraPosition=$Particles.Position) {isSetParameters=true, module=SetVariables_#8}
   &ViewMask(Border Percent Max=25, Border Percent Min=0) {module=ViewMask001}
   &SampleParticlesFromOtherEmitter(Velocity Application:ENiagara_DirectReadApplicationMode=::NewEnumerator0, Velocity Sampling=::NewEnumerator0)
   &ShapeLocation(
    MaskSpawnGroup=false
    Non Uniform Scale=&RandomRangeVector(Minimum=<0 0 0>, Maximum=<1 1 2>)
    Override Seed=false
    Sphere Radius=&RandomRangeFloat(Minimum=1, Maximum=25)
    UseOwnerScale=true
   )
  ]
  particleUpdateScript [
   &ParticleState()
   &KillParticlesInVolume(Sphere Radius=2500)
   &RegionMask(Region Falloff=1, Region Radius=8000, UseOffset=false, UseRotation=false)
   /SetVariables() {enabled=false, isSetParameters=true, module=SetVariables_#9}
   /SetVariables(Particles.GradientColor:Vector3f=&MakeVectorFromLinearColorRGB(Color=&MakeLinearColorFromVectorAndFloat(
      "Vector (RGB)"=&MakeVectorFromLinearColorRGB(Color=&ColorFromCurve(ColorCurve=DataInterface() {propertyValues {
          AlphaCurve {keys [<at -0.0023 1 RCIM_Cubic>,<at 0.9977 1 RCIM_Cubic>]}
          BlueCurve {keys [<at 0.1175 0.3005 RCIM_Cubic>,<at 0.2682 0.2705>,<at 0.4507 0.2122>,<at 0.5951 0.5271>,<at 0.7691 1.5>,<at 0.9071 0.521 RCIM_Cubic>]}
          GreenCurve {keys [<at 0.1175 0.1878 RCIM_Cubic>,<at 0.2682 0.2623>,<at 0.4507 0.2346>,<at 0.5951 0.4287>,<at 0.7691 0.4532>,<at 0.9071 0.4851 RCIM_Cubic>]}
          RedCurve {keys [<at 0.1175 0.3372 RCIM_Cubic>,<at 0.2682 0.7305>,<at 0.4507 0.7758>,<at 0.5951 0.7084>,<at 0.7691 0>,<at 0.9071 0.807 RCIM_Cubic>]}
         }}))
      "Float (Alpha)"=1
     ))) {enabled=false, isSetParameters=true, module=SetVariables_#10}
   /SetVariables(Particles.VisibilityTag=&MakeCustomIntFromBool(Bool=$Transient.FirstFrame, True INT=1, False INT=0)) {enabled=false, isSetParameters=true, module=SetVariables_#11}
   &PointForce(Force Falloff Distance=250000, Force Strength=-250, Module.Use Falloff=true, Module.UseLinearForce=false)
   &PointAttractionForce(
    Attraction Radius=400
    Attraction Strength=-1
    Attractor Position:Vector3f=$Emitter.CameraPosition
    Attractor Position Offset:Vector3f=<0 0 -500>
    Falloff Exponent=1
    Kill Within Radius=false
    Module.Use Falloff=true
    UseAttractorOffset=true
   ) {enabled=false, module=PointAttractionForce001}
   &WindForce(
    Alignment Falloff Start / End:Vector2f=<0 250>
    Frequency:NiagaraFloat=&RandomRangeFloat(Minimum=0.1, Maximum=0.2)
    Scale=1
    Wind Speed=<1 0 1>
    Wind Speed Scale=25
   )
   &AerodynamicDrag(
    Aerodynamic Drag:NiagaraFloat=&RandomRangeFloat(Minimum=0.01, Maximum=0.02)
    Lift Contribution=0
    Use Linear Drag=true
    Use Rotational Drag=true
   )
   &SolveForcesAndVelocity(
    Clamp Velocity=true
    Limit Acceleration=false
    Speed Limit:NiagaraFloat=&Multiply_Float(
     A=250
     B=&FloatFromCurve(
      FloatCurve=DataInterface() {propertyValues {Curve {keys [
          <at 0 0.5>
          {arriveTangent=1.255, interpMode=RCIM_Cubic, leaveTangent=1.255, tangentMode=RCTM_User, time=0.5, value=0.5}
          {arriveTangent=2.5018, interpMode=RCIM_Cubic, leaveTangent=2.5018, tangentMode=RCTM_User, time=1, value=2}
         ]}}}
      CurveIndex=$Particles.RegionMask.Float
     )
    )
   )
   &ViewMask(Border Percent Max=15)
   &ScaleColor(
    Scale Alpha=&Multiply_Float(
     A=&Multiply_Float(
      A=&FloatFromCurve(
       FloatCurve=DataInterface() {propertyValues {Curve {keys [
           {interpMode=RCIM_Cubic, tangentMode=RCTM_User, time=0, value=0}
           {interpMode=RCIM_Cubic, tangentMode=RCTM_User, time=0.35, value=1}
           {arriveTangent=-2.2965, interpMode=RCIM_Cubic, leaveTangent=-2.2965, tangentMode=RCTM_User, time=0.6, value=0.4}
           {interpMode=RCIM_Cubic, tangentMode=RCTM_User, time=1, value=0}
          ]}}}
       Scale Curve=&Multiply_Float(
        A=&RandomRangeFloat(Minimum=1, Maximum=2)
        B=&FloatFromCurve(FloatCurve=DataInterface() {propertyValues {Curve {keys [<at 0 1>,<at 1 0>]}}}, CurveIndex=$Particles.RegionMask.Float)
       )
      )
      B=&Waveform(
       "Waveform [1]"=::NewEnumerator0
       "[1] Phase Scale"=&RandomRangeFloat(Minimum=0.5, Maximum=2)
       "[1] Frequency"=&Multiply_Float(A=&RandomRangeFloat(Minimum=0.5, Maximum=1), B=$System.Main_FlickerSpeed)
       "[1] Amplitude Min/Max"=<0 0.1>
      )
     )
     B=&Multiply_Float(A=1, B=$System.Main_Alpha)
    )
    Scale RGB=&MakeVectorFromLinearColorRGB(Color=&Multiply_LinearColorByFloat(
      Linear Color=&ColorFromCurve(
       ColorCurve=DataInterface() {propertyValues {
         AlphaCurve {keys [<at 0 1>]}
         BlueCurve {keys [
           <at 0 0.0168 RCIM_Cubic>
           <at 0.3709 0.2957>
           {interpMode=RCIM_Cubic, leaveTangent=41.4639, time=0.748, value=0.9826}
           {interpMode=RCIM_Cubic, leaveTangent=-4.8359, time=0.9015, value=13.4257}
          ]}
         GreenCurve {keys [
           <at 0 0.0343 RCIM_Cubic>
           <at 0.3709 0.0669>
           {interpMode=RCIM_Cubic, leaveTangent=-0.2111, time=0.748, value=0}
           {arriveTangent=4.8898, interpMode=RCIM_Cubic, leaveTangent=4.8898, tangentMode=RCTM_User, time=0.9015, value=0}
          ]}
         RedCurve {keys [
           <at 0 0.147 RCIM_Cubic>
           <at 0.3709 0.0831>
           {interpMode=RCIM_Cubic, leaveTangent=78.6863, time=0.748, value=0.5267}
           {interpMode=RCIM_Cubic, tangentMode=RCTM_User, time=0.9015, value=25}
          ]}
        }}
       CurveIndex=$Particles.RegionMask.Float
      )
      Float=&RandomRangeFloat(Minimum=25, Maximum=25)
     ))
    ScaleA=true
    ScaleRGB=true
   )
   &DynamicMaterialParameters(
    Index 0 Param 2=1
    Index 0 Param 3=1
    Index 0 Param 4=15
    Module.Index0Float0Write=true
    Module.Index0Float1Write=true
    Module.Index0Float2Write=true
    Module.Index0Float3Write=true
   )
   /Niagara/Modules/Update/Camera/MaintainInCameraParticleScale(Render Target Width=1.5) {module=MaintainInCameraParticleScale001}
   &ScaleSpriteSizeBySpeed(Max Scale Factor=<2 3>)
   &ScaleSpriteSize(Scale Factor=&Vector2DFromFloat(Value=&FloatFromCurve(
      FloatCurve=DataInterface() {propertyValues {Curve {keys [
          {interpMode=RCIM_Cubic, tangentMode=RCTM_User, time=0, value=0.25}
          {interpMode=RCIM_Cubic, tangentMode=RCTM_User, time=0.5, value=1}
          {tangentMode=RCTM_User, time=1, value=1}
         ]}}}
      Scale Curve=&Multiply_Float(
       A=&NormalizeDistanceRange(
        Start Position:NiagaraPosition=$Particles.Position
        End Position:NiagaraPosition=Linked() {linkedVariable {name=System.CameraPostion, type=Vector3f}}
        Distance:NiagaraFloat=1500
        Invert Normalized Range:NiagaraBool=false
       )
       B=&Multiply_Float(A=$Output.ViewMask.ViewMask, B=1)
      )
     )))
   &AlignSpriteToMeshOrientation(SetAlignment=true, SetFacing=true)
   &CameraOffset(Camera Offset Amount=50) {module=CameraOffset001}
  ]
  rendererClasses [&NiagaraSpriteRendererProperties]
  renderers [
   {rendererClass=&NiagaraSpriteRendererProperties}
   {rendererClass=&NiagaraSpriteRendererProperties, rendererIndex=1}
  ]
 }
 SparklesParent {
  simTarget=GPUComputeSim
  emitterUpdateScript [
   &EmitterState(
    Enable Distance Culling=true
    MaxDistance=25000
    Module.bApplySpawnCountScaleByDistance=true
    Module.bUseMaxDistance=true
    Module.bUseMinDistance=false
    Scalability Mode=::NewEnumerator1
    Scale Spawn Count=true
    Spawn Count Scale By Distance=DataInterface() {propertyValues {Curve {keys [<at 0 0>,<at 0.1 0>,<at 0.5 1>,<at 0.75 1>,<at 1 1>]}}}
   )
   &SpawnRate(SpawnRate=&Multiply_Float(A=1500, B=1), Use Spawn Probability=true)
  ]
  particleSpawnScript [
   &InitializeParticle(
    Lifetime Max=3
    Lifetime Min=2
    Lifetime Mode=::NewEnumerator1
    Sprite Size Mode=::NewEnumerator1
    Uniform Sprite Size Max=5
    Uniform Sprite Size Min=3
    UsePositionOffset=false
    Write Material Random=true
   ) {module=InitializeParticle001}
   /Game/Effects/Niagara/Modules/Spawn/FN_Spawn_GBuffer_Color() {enabled=false}
   /Niagara/Modules/Update/Forces/CurlNoiseForce(Noise Frequency=150, Pan Noise=true, Pan Noise Field=<0 0 -5>) {enabled=false, module=CurlNoiseForce001}
   &ApplyInitialForces() {enabled=false}
   &CameraOffset(Camera Offset Amount=50)
   &ShapeLocation(MaskSpawnGroup=false, Override Seed=false, Sphere Radius=15000, UseOwnerScale=true)
   /Niagara/Modules/Math/MoveToNearestDistanceFieldSurface_GPU(Kill by Distance=50, bKillByDistance=false) {enabled=false}
   /SetVariables(Particles.InitialPos:NiagaraPosition=$Particles.Position) {isSetParameters=true, module=SetVariables_#8}
   /Niagara/Modules/Math/PlaceParticlesOnDepthBuffer_GPU(Max Z Depth=15000, Module.Use Max Z Depth=true)
   &ViewMask(Border Percent Max=25, Border Percent Min=0) {module=ViewMask001}
   &KillParticles(Kill Particles=&SetBoolByFloatComparison(A=1, B=$Output.ViewMask001.ViewMask, Comparison Type=::NewEnumerator6)) {module=KillParticles002}
   &KillParticles(Kill Particles=&SetBoolByFloatComparison(
     A=&VectorLength(Vector=&Subtract_Vector(A=$Particles.InitialPos, B=Linked() {linkedVariable {name=Particles.Position, type=NiagaraPosition}}))
     B=15000
    ))
   &KillParticles(Kill Particles=&Boolean_NotOperation(Bool=$Output.PlaceParticlesOnDepthBuffer_GPU.ValidSample)) {module=KillParticles001}
   /Game/Effects/Niagara/Modules/Spawn/FN_Spawn_GBuffer_Color_WithinRange(ColorMaxRange=1) {enabled=false}
  ]
  particleUpdateScript [
   &ParticleState()
   &KillParticlesInVolume(Sphere Radius=1000)
   &RegionMask(Region Falloff=1, Region Radius=8000, UseOffset=false, UseRotation=false)
   /SetVariables() {enabled=false, isSetParameters=true, module=SetVariables_#9}
   /SetVariables(Particles.GradientColor:Vector3f=&MakeVectorFromLinearColorRGB(Color=&MakeLinearColorFromVectorAndFloat(
      "Vector (RGB)"=&MakeVectorFromLinearColorRGB(Color=&ColorFromCurve(ColorCurve=DataInterface() {propertyValues {
          AlphaCurve {keys [<at -0.0023 1 RCIM_Cubic>,<at 0.9977 1 RCIM_Cubic>]}
          BlueCurve {keys [<at 0.1175 0.3005 RCIM_Cubic>,<at 0.2682 0.2705>,<at 0.4507 0.2122>,<at 0.5951 0.5271>,<at 0.7691 1.5>,<at 0.9071 0.521 RCIM_Cubic>]}
          GreenCurve {keys [<at 0.1175 0.1878 RCIM_Cubic>,<at 0.2682 0.2623>,<at 0.4507 0.2346>,<at 0.5951 0.4287>,<at 0.7691 0.4532>,<at 0.9071 0.4851 RCIM_Cubic>]}
          RedCurve {keys [<at 0.1175 0.3372 RCIM_Cubic>,<at 0.2682 0.7305>,<at 0.4507 0.7758>,<at 0.5951 0.7084>,<at 0.7691 0>,<at 0.9071 0.807 RCIM_Cubic>]}
         }}))
      "Float (Alpha)"=1
     ))) {enabled=false, isSetParameters=true, module=SetVariables_#10}
   /SetVariables(Particles.VisibilityTag=&MakeCustomIntFromBool(Bool=$Transient.FirstFrame, True INT=1, False INT=0)) {enabled=false, isSetParameters=true, module=SetVariables_#11}
   &PointForce(Force Falloff Distance=250000, Force Strength=-15000, Module.Use Falloff=true, Module.UseLinearForce=false)
   &PointAttractionForce(
    Attraction Radius=400
    Attraction Strength=-1
    Attractor Position:Vector3f=$Emitter.CameraPosition
    Attractor Position Offset:Vector3f=<0 0 -500>
    Falloff Exponent=1
    Kill Within Radius=false
    Module.Use Falloff=true
    UseAttractorOffset=true
   ) {enabled=false, module=PointAttractionForce001}
   &WindForce(
    Frequency:NiagaraFloat=&RandomRangeFloat(Minimum=0.1, Maximum=2)
    Scale=250
    Wind Speed=<25 0 50>
    Wind Speed Scale=25
   )
   &AerodynamicDrag(
    Aerodynamic Drag:NiagaraFloat=&RandomRangeFloat(Minimum=0.15, Maximum=0.5)
    Lift Contribution=0
    Use Linear Drag=true
    Use Rotational Drag=true
   )
   &SolveForcesAndVelocity(
    Clamp Velocity=true
    Limit Acceleration=false
    Speed Limit:NiagaraFloat=&Multiply_Float(
     A=150
     B=&FloatFromCurve(
      FloatCurve=DataInterface() {propertyValues {Curve {keys [
          <at 0 0.5>
          {arriveTangent=1.255, interpMode=RCIM_Cubic, leaveTangent=1.255, tangentMode=RCTM_User, time=0.5, value=1}
          {arriveTangent=2.5018, interpMode=RCIM_Cubic, leaveTangent=2.5018, tangentMode=RCTM_User, time=1, value=3}
         ]}}}
      CurveIndex=$Particles.RegionMask.Float
     )
    )
   )
   &CalculateAccurateVelocity()
   &ViewMask(Border Percent Max=15)
   &ScaleColor(
    Scale Alpha=&Multiply_Float(
     A=&Multiply_Float(
      A=&FloatFromCurve(
       FloatCurve=DataInterface() {propertyValues {Curve {keys [
           {interpMode=RCIM_Cubic, tangentMode=RCTM_User, time=0, value=0}
           {interpMode=RCIM_Cubic, tangentMode=RCTM_User, time=0.35, value=1}
           {arriveTangent=-2.2965, interpMode=RCIM_Cubic, leaveTangent=-2.2965, tangentMode=RCTM_User, time=0.6, value=0.4}
           {interpMode=RCIM_Cubic, tangentMode=RCTM_User, time=1, value=0}
          ]}}}
       Scale Curve=&Multiply_Float(
        A=&RandomRangeFloat(Minimum=2, Maximum=5)
        B=&FloatFromCurve(FloatCurve=DataInterface() {propertyValues {Curve {keys [<at 0 1>,<at 1 0>]}}}, CurveIndex=$Particles.RegionMask.Float)
       )
      )
      B=&Waveform(
       "Waveform [1]"=::NewEnumerator0
       "[1] Phase Scale"=&Multiply_Float(A=&RandomRangeFloat(Minimum=0.5, Maximum=1), B=$System.Main_FlickerSpeed)
       "[1] Frequency"=&RandomRangeFloat(Minimum=1, Maximum=2)
       "[1] Amplitude Min/Max"=<0 0.1>
      )
     )
     B=&Multiply_Float(A=0.5, B=$System.Main_Alpha)
    )
    Scale RGB=&MakeVectorFromLinearColorRGB(Color=&Multiply_LinearColorByFloat(
      Linear Color=&ColorFromCurve(
       ColorCurve=DataInterface() {propertyValues {
         AlphaCurve {keys [<at 0 1>]}
         BlueCurve {keys [
           <at 0 0.0168 RCIM_Cubic>
           <at 0.3709 0.2957>
           {interpMode=RCIM_Cubic, leaveTangent=41.4639, time=0.748, value=0.9826}
           {interpMode=RCIM_Cubic, leaveTangent=-4.8359, time=0.9015, value=13.4257}
          ]}
         GreenCurve {keys [
           <at 0 0.0343 RCIM_Cubic>
           <at 0.3709 0.0669>
           {interpMode=RCIM_Cubic, leaveTangent=-0.2111, time=0.748, value=0}
           {arriveTangent=4.8898, interpMode=RCIM_Cubic, leaveTangent=4.8898, tangentMode=RCTM_User, time=0.9015, value=0}
          ]}
         RedCurve {keys [
           <at 0 0.147 RCIM_Cubic>
           <at 0.3709 0.0831>
           {interpMode=RCIM_Cubic, leaveTangent=78.6863, time=0.748, value=0.5267}
           {interpMode=RCIM_Cubic, tangentMode=RCTM_User, time=0.9015, value=25}
          ]}
        }}
       CurveIndex=$Particles.RegionMask.Float
      )
      Float=&RandomRangeFloat(Minimum=500, Maximum=2500)
     ))
    ScaleA=true
    ScaleRGB=true
   )
   &DynamicMaterialParameters(
    Index 0 Param 2=1
    Index 0 Param 3=1
    Index 0 Param 4=15
    Module.Index0Float0Write=true
    Module.Index0Float1Write=true
    Module.Index0Float2Write=true
    Module.Index0Float3Write=true
   )
   &MaintainInCameraParticleScale() {enabled=false}
   &MaintainInCameraParticleScale(Render Target Width=1.5) {enabled=false, module=MaintainInCameraParticleScale001}
   &ScaleSpriteSizeBySpeed(Max Scale Factor=<2 3>)
   &ScaleSpriteSize(Scale Factor=&Vector2DFromFloat(Value=&FloatFromCurve(
      FloatCurve=DataInterface() {propertyValues {Curve {keys [
          {interpMode=RCIM_Cubic, tangentMode=RCTM_User, time=0, value=0.25}
          {interpMode=RCIM_Cubic, tangentMode=RCTM_User, time=0.5, value=1}
          {tangentMode=RCTM_User, time=1, value=1}
         ]}}}
      Scale Curve=&Multiply_Float(
       A=&NormalizeDistanceRange(
        Start Position:NiagaraPosition=$Particles.Position
        End Position:NiagaraPosition=Linked() {linkedVariable {name=System.CameraPostion, type=Vector3f}}
        Distance:NiagaraFloat=1500
        Invert Normalized Range:NiagaraBool=false
       )
       B=&Multiply_Float(A=$Output.ViewMask.ViewMask, B=1)
      )
     )))
   &AlignSpriteToMeshOrientation(SetAlignment=true, SetFacing=true)
   &CameraOffset(Camera Offset Amount=50) {module=CameraOffset001}
  ]
  rendererClasses [&NiagaraSpriteRendererProperties]
  renderers [
   {rendererClass=&NiagaraSpriteRendererProperties}
   {rendererClass=&NiagaraSpriteRendererProperties, rendererIndex=1}
  ]
 }
}
summary {
 systemName=NS_MaskFortress_Smoke_MidDistance
 emitters [
  {bEnabled=false, emitterName=Ground_MidDist, simTarget=GPUComputeSim, rendererClasses [&NiagaraMeshRendererProperties]}
  {emitterName=SparklesParent, simTarget=GPUComputeSim, rendererClasses [&NiagaraSpriteRendererProperties]}
  {emitterName=Sky, simTarget=GPUComputeSim, rendererClasses [&NiagaraMeshRendererProperties]}
  {emitterName=Sparkles, simTarget=GPUComputeSim, rendererClasses [&NiagaraSpriteRendererProperties]}
  {emitterName=Ground_Lines, simTarget=GPUComputeSim, rendererClasses [&NiagaraMeshRendererProperties]}
 ]
 userVariables [User.Landscape {
   type=NiagaraDataInterfaceLandscape
   description=""
   defaultValue {
    struct=NiagaraExt_VariableValue_DataInterface
    value {
     dataInterface=&self:NiagaraDataInterfaceLandscape_0
     dataInterfaceClass=&NiagaraDataInterfaceLandscape
    }
   }
  }]
}
