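# Translucent, two-sided unlit material for spiral lightning meshes in the MaskFortress effects, usable on Niagara mesh particles.
# Panning noise, optionally time-stepped, distorts lightning texture UVs, while a sine-based moving gradient supplies a mask.
# SplineThicken and normal offsets shape the mesh, and curve-atlas or gradient colors receive an HDR boost.
# The UseInNiagara switch swaps material parameters for dynamic parameters and particle color.
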Material /BRMapCh6/FX/MaskFortress/Materials/M_MaskFortress_Lighting_Spiral_Mesh
graph {
 properties {
  bAutomaticallySetUsageInEditor=false
  bUsedWithNiagaraMeshParticles=true
  blendMode=BLEND_Translucent
  shadingModel=MSM_Unlit
  twoSided=true
  usedShadingModels=Unlit
 }
 nodes [
  n1 = DynamicParameter() {parameterIndex=1, defaultValue=<rgba 1 1 25 25>, paramNames [TextureTile_X, Texture_TileY, WidthBase, WidthTip]}
  n2 = ScalarParameter() {defaultValue=10, parameterName=WidthBase}
  n3 = OneMinus(None=ComponentMask(None=TextureCoordinate()) {g=true})
  n4 = DynamicParameter() {parameterIndex=1, defaultValue=<rgba 1 1 25 25>, paramNames [TextureTile_X, Texture_TileY, WidthBase, WidthTip]}
  n5 = ScalarParameter() {defaultValue=10, parameterName=WidthTip}
  n6 = MaterialFunctionCall(
   WidthBase=Multiply(A=StaticSwitchParameter(True=@n1.WidthBase, False=@n2) {parameterName=UseInNiagara?}, B=@n3)
   WidthTip=Multiply(A=@n3, B=StaticSwitchParameter(True=@n4.WidthTip, False=@n5) {parameterName=UseInNiagara?})
  ) {materialFunction=&SplineThicken}
  n7 = ComponentMask(None=TextureCoordinate()) {g=true}
  n8 = OneMinus(None=@n7)
  n9 = Multiply(
   A=VertexNormalWS()
   B=Multiply(A=DynamicParameter() {parameterIndex=2, paramNames [MainScaler, PannerMask_Speed, Distortion, Param4]}.MainScaler, B=Multiply(A=@n8, B=@n7))
  )
  Add(A=@n6.WorldPositionOffset, B=@n9)
  DynamicParameter() {parameterIndex=2, paramNames [MainScaler, PannerMask_Speed, Distortion, Param4]}
  DynamicParameter() {parameterIndex=3}
  n10 = NamedRerouteUsage()
  LinearInterpolate(B=Power(Base=Multiply(A=@n10, B=@n10)), Alpha=VertexColor().R)
  n11 = TextureCoordinate()
  n12 = Multiply(A=NamedRerouteUsage(), B=OneMinus(None=ComponentMask(None=@n11) {g=true}))
  n13 = CurveAtlasRowParameter(CurveTime=@n12) {atlas=/BRMapCh6/FX/MaskFortress/CA_MaskFortress, curve=&/Curve_MaskFortress, parameterName=Channel_Color}
  n14 = Clamp(None=@n13)
  n15 = CurveAtlasRowParameter(CurveTime=@n12) {
   atlas=&CA_MaskFortress
   curve=&/Curve_MaskFortress_HDR
   defaultValue=2
   parameterName=Channel_HDR_Intensity
  }
  MaterialFunctionCall(LDR Input=@n14, HDR Intensity Multiplier=Multiply(A=@n15)) {materialFunction=&DeriveHDRfromLDR}
  n16 = TextureCoordinate()
  n17 = VertexColor()
  n18 = DynamicParameter() {defaultValue=<rgba 1 25 1 1>, paramNames [Tile_X, HDR_Power, DistortSpeed_X, DistortSpeed_Y]}
  n19 = ScalarParameter() {defaultValue=1, parameterName=Dynamic V Scale}
  n20 = VertexColor()
  n21 = AppendVector(
   A=Constant()
   B=StaticSwitchParameter(True=Multiply(A=@n17.G, B=@n18.Tile_X), False=Multiply(A=@n19, B=@n20.G)) {parameterName=UseInNiagara?}
  )
  Multiply(A=@n16, B=@n21)
  n22 = TextureCoordinate()
  n23 = NamedRerouteUsage()
  Multiply(A=@n22, B=@n23)
  n24 = StaticSwitchParameter(
   True=DynamicParameter() {parameterIndex=2, paramNames [MainScaler, PannerMask_Speed, Distortion, Param4]}.Distortion
   False=ScalarParameter() {defaultValue=0.25, parameterName=DistortionAmount}
  ) {parameterName=UseInNiagara?}
  n25 = TextureCoordinate() {uTiling=0.46, vTiling=0.2476}
  n26 = Time()
  n27 = ScalarParameter() {defaultValue=1, parameterName=UVDistortSpeed}
  n28 = Divide(A=Ceil(None=Multiply(A=@n26, B=@n27) {constB=16})) {constB=4}
  n29 = NamedRerouteUsage()
  n30 = TextureCoordinate() {uTiling=0.46, vTiling=0.2476}
  n31 = Time()
  n32 = NamedRerouteUsage()
  n33 = TextureSampleParameter2D(UVs=StaticSwitchParameter(True=Panner(Coordinate=@n25, Time=@n28, Speed=@n29) {speedY=0.4}, False=Panner(Coordinate=@n30, Time=@n31, Speed=@n32) {speedY=0.4}) {parameterName=UseStepping?}) {
   parameterName=DistortionTexture
   samplerType=SAMPLERTYPE_LinearColor
   texture=/Game/Global/Textures/Default/Noise/TilingNoise06_N_uncompressed
  }
  n34 = ScalarParameter() {parameterName=X Speed}
  n35 = ScalarParameter() {defaultValue=1, parameterName=Y Speed}
  n36 = ScalarParameter() {defaultValue=1, group=Step Panner, parameterName=StepPan_Timespeed}
  n37 = AppendVector(
   A=ScalarParameter() {group=Step Panner, parameterName=StepPan_JumpAmount_X}
   B=ScalarParameter() {defaultValue=2, group=Step Panner, parameterName=StepPan_JumpAmount_Y}
  )
  n38 = AppendVector(
   A=ScalarParameter() {defaultValue=1.5, group=Step Panner, parameterName=XStepPan_NumberofStepsPerSecond}
   B=ScalarParameter() {defaultValue=1.5, group=Step Panner, parameterName=YStepPan_NumberofStepsPerSecond}
  )
  n39 = Panner(
   Coordinate=Add(A=@n16, B=@n21)
   Speed=LinearInterpolate(
    A=AppendVector(A=@n34, B=@n35)
    B=MaterialFunctionCall!(TimeSpeed=@n36, Jump Amount Per Step=@n37, Number of Steps Per Unit of Time=@n38, Time Offset) {materialFunction=/Engine/Functions/Engine_MaterialFunctions02/SteppingPannerTime}.Result
    Alpha=ScalarParameter() {group=Step Panner, parameterName=Use_StepPanner}
   )
  ) {bFractionalPart=true}
  n40 = Constant()
  n41 = MaterialFunctionCall(Float2=Add(A=ComponentMask(None=@n33.RGB) {g=true, r=true}, B=StaticSwitchParameter(True=@n39, False=@n40) {parameterName=UseStepping?}) {constB=0}) {materialFunction=&BreakOutFloat2Components}
  n42 = AppendVector(
   A=LinearInterpolate(A=@n41.R, B=Clamp(None=@n41.R), Alpha=ScalarParameter() {defaultValue=1, parameterName=Clamp Texture V})
   B=OneMinus(None=@n41.G)
  )
  n43 = Add(A=@n22, B=@n23)
  n44 = AppendVector(A=ScalarParameter() {defaultValue=16, parameterName=X_Offset}, B=Constant())
  n45 = TextureSampleParameter2D(UVs=Add(A=Multiply(A=@n24, B=@n42), B=Add(A=@n43, B=@n44))) {parameterName=TilingTexture, texture=/Game/Environments/Hexmap/HexClouds/Lightning/T_Hexmap_Lightning_02_E}
  Multiply(A=NamedRerouteUsage(), B=@n45.B)
  Multiply()
  n46 = ParticleColor()
  n47 = VertexColor()
  Multiply(A=@n45.G, B=StaticSwitchParameter(True=Multiply(A=@n46.A, B=@n47.R), False=VertexColor().B) {parameterName=UseInNiagara?})
  Multiply()
  n48 = DynamicParameter() {defaultValue=<rgba 1 25 1 1>, paramNames [Tile_X, HDR_Power, DistortSpeed_X, DistortSpeed_Y]}
  n49 = VectorParameter() {parameterName=DistortionSpeed, defaultValue=<rgba 0 0.25 0 1>}
  NamedRerouteDeclaration(None=StaticSwitchParameter(True=AppendVector(A=@n48.DistortSpeed_X, B=@n48.DistortSpeed_X), False=ComponentMask(None=@n49.RGB) {g=true, r=true}) {parameterName=UseInNiagara?}) {name=DistortSpeed, nodeColor=<rgba 0 1 0.8828 1>}
  n50 = DynamicParameter() {parameterIndex=1, defaultValue=<rgba 1 1 25 25>, paramNames [TextureTile_X, Texture_TileY, WidthBase, WidthTip]}
  n51 = Multiply(A=@n50.Texture_TileY, B=LinearInterpolate(Alpha=VertexColor().G) {constA=0.35, constB=0.6})
  n52 = ScalarParameter() {defaultValue=1, parameterName=TextureTile_X}
  n53 = ScalarParameter() {defaultValue=1, parameterName=TextureTile_Y}
  NamedRerouteDeclaration(None=StaticSwitchParameter(True=AppendVector(A=@n50.TextureTile_X, B=@n51), False=AppendVector(A=@n52, B=@n53)) {parameterName=UseInNiagara?}) {name=TextureTile, nodeColor=<rgba 1 0 0.3984 1>}
  n54 = VectorParameter() {parameterName=Lighting_Color, defaultValue=<rgba 0.7672 0 1 1>}
  n55 = VectorParameter() {parameterName=SecondaryGlowColor, defaultValue=<rgba 60 55.9389 21.5625 1>}
  n56 = Multiply(A=MaterialFunctionCall() {materialFunction=&LinearGradient}.VGradient)
  n57 = StaticSwitchParameter(
   True=Multiply(A=@n14, B=ParticleColor().RGB)
   False=LinearInterpolate(
    A=LinearInterpolate(A=@n54.RGB, B=@n55.RGB, Alpha=@n56) {constB=0}
    B=VectorParameter() {parameterName=Lighting_Color, defaultValue=<rgba 0.7672 0 1 1>}.RGB
    Alpha=@n45.G
   )
  ) {parameterName=UseInNiagara?}
  n58 = DynamicParameter() {defaultValue=<rgba 1 25 1 1>, paramNames [Tile_X, HDR_Power, DistortSpeed_X, DistortSpeed_Y]}
  n59 = ScalarParameter() {defaultValue=5, parameterName=HDR_Power}
  NamedRerouteDeclaration(None=MaterialFunctionCall(LDR Input=Clamp(None=@n57), HDR Intensity Multiplier=StaticSwitchParameter(True=@n58.HDR_Power, False=@n59) {parameterName=UseInNiagara?}) {materialFunction=&DeriveHDRfromLDR}.RemappedRange) {name=HDR Color, nodeColor=<rgba 0 1 0.9531 1>}
  n60 = DynamicParameter() {parameterIndex=2, paramNames [MainScaler, PannerMask_Speed, Distortion, Param4]}
  n61 = Multiply(A=ScalarParameter() {defaultValue=0.25, parameterName=MoveMask_Speed}, B=LinearInterpolate(Alpha=VertexColor().A) {constA=1, constB=1.25})
  n62 = Add(
   A=Multiply(A=StaticSwitchParameter(True=@n60.PannerMask_Speed, False=@n61) {parameterName=UseInNiagara?}, B=Time())
   B=ComponentMask(None=TextureCoordinate()) {g=true}
  )
  n63 = OneMinus(None=ConstantBiasScale(None=Sine(None=@n62) {period=2}))
  NamedRerouteDeclaration(None=SmoothStep(Value=Clamp(None=@n63)) {constMin=-0.5}) {name=Moving Gradient Mask Alpha, nodeColor=<rgba 1 0.6328 0 1>}
  StaticSwitchParameter(True=Multiply(A=@n6.WorldPositionOffset, B=@n9), False=@n6.WorldPositionOffset) {parameterName=UseInNiagara?}
  VertexColor()
  VertexColor()
 ]
}
