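# Translucent unlit smoke material for Mask Fortress FX spline meshes, with a UseInNiagara switch swapping scalar parameters for particle dynamic inputs.
# A panning plasma texture drives curve-atlas HDR color and opacity, shaped by edge gradients, fresnel, erosion, moving sine masks and distance fades.
# World position offset combines SplineThicken width and noise displacement, optionally conforming to the landscape through a runtime virtual texture heightmap.
# A CheapShader switch provides simplified color and opacity.
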Material /BRMapCh6/FX/MaskFortress/Materials/M_MaskFortress_Smoke_Splines
graph {
 properties {
  bAutomaticallySetUsageInEditor=false
  bUsedWithNiagaraMeshParticles=true
  blendMode=BLEND_Translucent
  shadingModel=MSM_Unlit
  translucencyLightingMode=TLM_VolumetricPerVertexNonDirectional
  twoSided=true
  usedShadingModels=Unlit
 }
 nodes [
  Constant()
  n1 = ScalarParameter() {defaultValue=1, parameterName=SpeedMultVertex}
  LinearInterpolate(A=@n1, B=Multiply(A=@n1, B=NamedRerouteUsage()), Alpha=NamedRerouteUsage())
  n2 = PixelDepth()
  n3 = Constant()
  n4 = ScalarParameter() {defaultValue=1500, parameterName=FadeDistance}
  NamedRerouteDeclaration(None=OneMinus(None=SphereMask(A=@n2.R, B=@n3, Radius=@n4) {hardnessPercent=0})) {name=FX_DistanceFade_PerPixel, nodeColor=<rgba 0 1 0.6719 1>}
  NamedRerouteDeclaration(None=ScalarParameter() {defaultValue=1, parameterName=DistanceValueMultiplier}) {name=DistanceValueMultiplier, nodeColor=<rgba 1 0.6797 0 1>}
  n5 = AppendVector(
   A=ComponentMask(None=VectorParameter() {parameterName=ConformMult, defaultValue=<rgba 0 0 1 1>}.RGB) {g=true, r=true}
   B=DynamicParameter() {parameterIndex=2, paramNames [OffsetTexture, PannerMask_Speed, MainTexture_Speed, RVT_Conform]}.RVT_Conform
  )
  n6 = WorldPosition()
  n7 = RuntimeVirtualTextureSample!(UVs, World Position=@n6.XYZ) {materialType=WorldHeight, virtualTexture=/Game/Athena/Apollo/Maps/Landscape/VT/RVT_Apollo_Landscape_Height}
  n8 = ComponentMask(None=@n6.XYZ) {b=true}
  n9 = ComponentMask(None=ObjectBounds()) {b=true}
  n10 = Subtract(A=Add(A=Subtract(A=@n7.WorldHeight, B=@n8), B=@n9), B=Multiply(A=@n9, B=ScalarParameter() {defaultValue=1, parameterName=OffsetHeight}))
  NamedRerouteDeclaration(None=ComponentMask(None=Multiply(A=@n5, B=@n10)) {b=true}) {name=RVT_Heightmap, nodeColor=<rgba 0.5469 0 1 1>}
  n11 = VertexColor()
  n12 = StaticSwitchParameter(
   True=DynamicParameter() {parameterIndex=2, paramNames [OffsetTexture, PannerMask_Speed, MainTexture_Speed, RVT_Conform]}.PannerMask_Speed
   False=Multiply(A=ScalarParameter() {defaultValue=0.25, parameterName=MoveMask_Speed}, B=LinearInterpolate(Alpha=@n11.A) {constA=1, constB=1.25})
  ) {parameterName=UseInNiagara?}
  n13 = Time()
  n14 = TextureCoordinate()
  n15 = Sine(None=Add(A=Multiply(A=@n12, B=@n13), B=ComponentMask(None=@n14) {g=true})) {period=2}
  n16 = ConstantBiasScale(None=@n15)
  NamedRerouteDeclaration(None=Clamp(None=OneMinus(None=@n16))) {name=Moving Gradient Mask Color, nodeColor=<rgba 0.6771 0.2877 0.0225 1>}
  n17 = MaterialFunctionCall(Exponent=ScalarParameter() {defaultValue=1, parameterName="Exponent (S)"}) {materialFunction=&AxisAlignedFresnel}
  NamedRerouteDeclaration(None=@n17.Fresnel) {name=Niagara_Fresnel, nodeColor=<rgba 1 0.1875 0 1>}
  n18 = MaterialFunctionCall(UV Channel=TextureCoordinate() {vTiling=0.5}) {materialFunction=&LinearGradient}
  n19 = OneMinus(None=@n18.UGradient)
  n20 = OneMinus(None=@n18.VGradient)
  n21 = TextureCoordinate()
  n22 = DynamicParameter() {parameterIndex=2, paramNames [OffsetTexture, PannerMask_Speed, MainTexture_Speed, RVT_Conform]}
  n23 = VertexColor()
  n24 = AppendVector(
   A=ComponentMask(None=@n21) {r=true}
   B=Add(A=ComponentMask(None=@n21) {g=true}, B=StaticSwitchParameter(True=@n22.OffsetTexture, False=@n23.A) {parameterName=UseInNiagara?})
  )
  n25 = VectorParameter() {parameterName=PannerUVMult, defaultValue=<rgba 1 1 0 1>}
  n26 = ComponentMask(None=@n25.RGB) {g=true}
  n27 = DynamicParameter() {defaultValue=<rgba 100 100 1 1>, paramNames [WidthBase,WidthTip,Erode,TextureY]}
  n28 = AppendVector(
   A=ComponentMask(None=@n25.RGB) {r=true}
   B=StaticSwitchParameter(True=Multiply(A=@n26, B=@n27.TextureY), False=@n26) {parameterName=UseInNiagara?}
  )
  n29 = AppendVector(
   A=Constant()
   B=DynamicParameter() {parameterIndex=2, paramNames [OffsetTexture, PannerMask_Speed, MainTexture_Speed, RVT_Conform]}.MainTexture_Speed
  )
  n30 = AppendVector(A=Constant(), B=ScalarParameter() {defaultValue=1, parameterName=SpeedMultVertex})
  n31 = TextureSample(UVs=Panner(Coordinate=Multiply(A=@n24, B=@n28), Speed=StaticSwitchParameter(True=@n29, False=@n30) {parameterName=UseInNiagara?}) {speedY=1}) {samplerType=SAMPLERTYPE_LinearGrayscale, texture=/Game/Effects/Fort_Effects/Emotes/UltraLight/Texture/T_Plasma_A_B_Greyscale}
  n32 = Multiply(A=Multiply(A=Multiply(A=@n18.UGradient, B=@n19), B=Multiply(A=@n18.VGradient, B=@n20)), B=@n31.G)
  NamedRerouteDeclaration(None=@n32) {name=Niagara_Texture, nodeColor=<rgba 0 0.1328 1 1>}
  n33 = WorldPosition()
  n34 = VectorParameter() {parameterName=SphereLocation, defaultValue=<rgba 0 0 0 1>}
  n35 = ScalarParameter() {defaultValue=25000, parameterName="Radius (S)"}
  n36 = ScalarParameter() {parameterName=Hardness}
  NamedRerouteDeclaration(None=Multiply(A=ScalarParameter() {defaultValue=1, parameterName=SphereEmissive}, B=SphereMask(A=@n33.XYZ, B=@n34.RGB, Radius=@n35, Hardness=@n36)) {constB=250}) {name=SphereMask_Glow, nodeColor=<rgba 0.1484 1 0 1>}
  n37 = ScalarParameter() {defaultValue=2500, parameterName=FarFadeout_Distance}
  n38 = ScalarParameter() {defaultValue=-2500, parameterName=FarFadeout_Distance_Offset}
  NamedRerouteDeclaration(None=OneMinus(None=MaterialFunctionCall!(Blend Range=@n37, Start Offset=@n38) {materialFunction=/Engine/Functions/Engine_MaterialFunctions03/Particles/Distance_Blend}.Result)) {name=DistanceFade, nodeColor=<rgba 0.9155 0.651 1 1>}
  n39 = DynamicParameter() {parameterIndex=1, defaultValue=<rgba 0.002 0.002 0.002 1>, paramNames [Curl_Speed, Curl_Scale, Curl_Intensity, CurlNoiseMult]}
  n40 = WorldPosition()
  n41 = ActorPositionWS()
  n42 = DynamicParameter() {parameterIndex=1, defaultValue=<rgba 0.002 0.002 0.002 1>, paramNames [Curl_Speed, Curl_Scale, Curl_Intensity, CurlNoiseMult]}
  n43 = Time()
  n44 = DynamicParameter() {parameterIndex=1, defaultValue=<rgba 0.002 0.002 0.002 1>, paramNames [Curl_Speed, Curl_Scale, Curl_Intensity, CurlNoiseMult]}
  n45 = Divide(A=Subtract(A=Subtract(A=@n40.XYZ, B=@n41), B=Multiply(A=@n42.Curl_Speed, B=@n43)), B=@n44.Curl_Scale)
  NamedRerouteDeclaration(None=Multiply(A=@n39.Curl_Intensity, B=VectorNoise(World Position=@n45) {noiseFunction=VNF_VectorALU})) {name=VectorNoiseNiagara, nodeColor=<rgba 0.3153 0.9234 1 1>}
  n46 = Time() {bIgnorePause=true, bOverride_Period=true, period=1000}
  n47 = Multiply(A=VectorParameter() {parameterName=PanSpeed, defaultValue=<rgba 250 1500 -500 1>}.RGB, B=NamedRerouteUsage())
  n48 = Multiply(
   A=VectorParameter() {parameterName=CurlNoise_Tiling, defaultValue=<rgba 0.02 0.02 0.02 1>}.RGB
   B=Add(A=Multiply(A=@n46, B=@n47) {constB=250}, B=WorldPosition().XYZ)
  )
  NamedRerouteDeclaration(None=Multiply(A=VectorNoise(World Position=@n48) {noiseFunction=VNF_CurlALU}, B=ScalarParameter() {defaultValue=1, parameterName=WPOCurlAmount})) {name=VectorNoise, nodeColor=<rgba 0.3153 0.9234 1 1>}
  n49 = ScalarParameter() {defaultValue=1, parameterName=SphereAlpha}
  n50 = SphereMask(
   A=WorldPosition().XYZ
   B=VectorParameter() {parameterName=SphereLocation, defaultValue=<rgba 0 0 0 1>}.RGB
   Radius=ScalarParameter() {defaultValue=80, parameterName=Radius Alpha}
   Hardness=ScalarParameter() {parameterName=Hardness Alpha}
  )
  NamedRerouteDeclaration(None=Clamp(None=Multiply(A=@n49, B=@n50) {constB=250})) {name=SphereMask_Alpha, nodeColor=<rgba 0.1484 1 0.3913 1>}
  n51 = MaterialFunctionCall(
   Blend Range=ScalarParameter() {defaultValue=2500, parameterName=CloseFadeout_Distance}
   Start Offset=ScalarParameter() {defaultValue=-2500, parameterName=CloseFadeout_Distance_Offset}
  ) {materialFunction=&Distance_Blend}
  NamedRerouteDeclaration(None=@n51.Result) {name=Close_DistanceFade, nodeColor=<rgba 0.8285 0.2917 1 1>}
  n52 = OneMinus(None=ConstantBiasScale(None=@n15))
  NamedRerouteDeclaration(None=SmoothStep(Value=Clamp(None=@n52)) {constMin=-0.5}) {name=Moving Gradient Mask Alpha, nodeColor=<rgba 1 0.6328 0 1>}
  n53 = NamedRerouteUsage()
  n54 = Multiply(A=Multiply(A=@n32, B=ScalarParameter() {defaultValue=1, parameterName=FinalOpacity}), B=@n17.Fresnel)
  NamedRerouteDeclaration(None=Multiply(A=NamedRerouteUsage(), B=Multiply(A=@n53, B=@n54))) {name=DistanceFading_Complete, nodeColor=<rgba 0.8513 0.4325 1 1>}
  NamedRerouteUsage()
  ObjectPositionWS()
  OneMinus(None=@n51.Result)
  ParticlePositionWS()
  Power(Base=NamedRerouteUsage(), Exp=ScalarParameter() {defaultValue=2, parameterName=Inner_Sphere_Opacity_Exponent})
  n55 = NamedRerouteUsage()
  n56 = ParticleColor()
  n57 = Multiply(A=NamedRerouteUsage(), B=StaticSwitchParameter(True=Constant() {r=1}, False=NamedRerouteUsage()) {parameterName=UseInNiagara?})
  n58 = MaterialFunctionCall!(
   Dynamic Param=DynamicParameter() {defaultValue=<rgba 100 100 1 1>, paramNames [WidthBase,WidthTip,Erode,TextureY]}.Erode
   Alpha=Multiply(A=NamedRerouteUsage(), B=NamedRerouteUsage())
  ) {materialFunction=/Game/Effects/Fort_Effects/Materials/_Functions/Erode_Alpha}
  n59 = Constant() {r=1}
  n60 = ScalarParameter() {defaultValue=1, parameterName=FinalOpacity2}
  n61 = Constant() {r=0.25}
  n62 = Multiply(
   A=Multiply(A=Clamp(None=@n57), B=StaticSwitchParameter(True=@n58.Result, False=@n59) {parameterName=UseInNiagara?})
   B=LinearInterpolate(A=Multiply(A=@n60, B=@n61), B=@n60, Alpha=NamedRerouteUsage())
  )
  StaticSwitchParameter(True=StaticSwitchParameter(True=Multiply(A=@n55, B=@n56.A), False=Multiply(A=@n56.A, B=@n62)) {parameterName=CheapShader?}, False=@n62) {parameterName=UseInNiagara?}
  n63 = Multiply(
   A=VectorParameter() {parameterName=FogColor, defaultValue=<rgba 0 0 0 1>}.RGB
   B=ScalarParameter() {defaultValue=1, parameterName=FakeFogAmount}
  )
  n64 = NamedRerouteUsage()
  n65 = TextureCoordinate()
  n66 = Multiply(A=NamedRerouteUsage(), B=OneMinus(None=ComponentMask(None=@n65) {g=true}))
  n67 = CurveAtlasRowParameter(CurveTime=@n66) {atlas=/BRMapCh6/FX/MaskFortress/CA_MaskFortress, curve=&/Curve_MaskFortress, parameterName=Channel_Color}
  n68 = CurveAtlasRowParameter(CurveTime=@n66) {
   atlas=&CA_MaskFortress
   curve=&/Curve_MaskFortress_HDR
   defaultValue=2
   parameterName=Channel_HDR_Intensity
  }
  n69 = DynamicParameter() {parameterIndex=3, defaultValue=<rgba 100 100 1 1>, paramNames [Emissive, HDR, Curve_or_ParticleColor, Param4]}
  n70 = ScalarParameter() {defaultValue=25, parameterName="HDR Intensity Multiplier (S)"}
  n71 = MaterialFunctionCall(
   LDR Input=Clamp(None=@n67)
   HDR Intensity Multiplier=Multiply(A=@n68, B=StaticSwitchParameter(True=@n69.HDR, False=@n70) {parameterName=UseInNiagara?})
  ) {materialFunction=&DeriveHDRfromLDR}
  n72 = DynamicParameter() {parameterIndex=3, defaultValue=<rgba 100 100 1 1>, paramNames [Emissive, HDR, Curve_or_ParticleColor, Param4]}
  n73 = LinearInterpolate(A=@n71.RemappedRange, B=ParticleColor().RGB, Alpha=@n72.Curve_or_ParticleColor)
  n74 = DynamicParameter() {parameterIndex=3, defaultValue=<rgba 100 100 1 1>, paramNames [Emissive, HDR, Curve_or_ParticleColor, Param4]}
  n75 = ScalarParameter() {defaultValue=1, parameterName=Emissive}
  n76 = Multiply(
   A=Multiply(A=StaticSwitchParameter(True=@n73, False=@n71.RemappedRange) {parameterName=UseInNiagara?}, B=@n31.G)
   B=Multiply(A=StaticSwitchParameter(True=@n74.Emissive, False=@n75) {parameterName=UseInNiagara?}, B=NamedRerouteUsage())
  )
  n77 = NamedRerouteUsage()
  n78 = ComponentMask(None=TextureCoordinate()) {g=true}
  n79 = Add(
   A=StaticSwitchParameter(True=Constant(), False=LinearInterpolate(B=@n63, Alpha=@n64)) {parameterName=UseInNiagara?}
   B=Multiply(A=Multiply(A=@n76, B=@n77), B=OneMinus(None=@n78))
  )
  n80 = Multiply(A=ParticleColor().RGB, B=@n79)
  StaticSwitchParameter(
   True=ParticleColor().RGB
   False=MaterialFunctionCall(Emissive=StaticSwitchParameter(True=@n80, False=@n79) {parameterName=UseInNiagara?}) {materialFunction=&MF_FX_EmissiveAdjustments}.Result
  ) {parameterName=CheapShader?}
  n81 = ScalarParameter() {parameterName="WidthBase (S)"}
  n82 = Clamp(None=LinearInterpolate(Alpha=NamedRerouteUsage()) {constA=1, constB=0})
  n83 = NamedRerouteUsage()
  n84 = Clamp(None=Multiply(A=NamedRerouteUsage(), B=Multiply(A=@n83)))
  n85 = StaticSwitchParameter(
   True=DynamicParameter() {defaultValue=<rgba 100 100 1 1>, paramNames [WidthBase,WidthTip,Erode,TextureY]}.WidthBase
   False=Multiply(A=Multiply(A=@n81, B=@n82), B=@n84)
  ) {parameterName=UseInNiagara?}
  n86 = ScalarParameter() {parameterName="WidthTip (S)"}
  n87 = StaticSwitchParameter(
   True=DynamicParameter() {defaultValue=<rgba 100 100 1 1>, paramNames [WidthBase,WidthTip,Erode,TextureY]}.WidthTip
   False=Multiply(A=Multiply(A=@n86, B=@n82), B=@n84)
  ) {parameterName=UseInNiagara?}
  n88 = StaticSwitchParameter(True=NamedRerouteUsage(), False=Multiply(A=VertexNormalWS(), B=NamedRerouteUsage())) {parameterName=UseInNiagara?}
  n89 = ComponentMask(None=@n88) {g=true, r=true}
  n90 = DynamicParameter() {parameterIndex=1, defaultValue=<rgba 0.002 0.002 0.002 1>, paramNames [Curl_Speed, Curl_Scale, Curl_Intensity, CurlNoiseMult]}
  n91 = StaticSwitchParameter(True=AppendVector(A=Multiply(A=@n89, B=@n90.CurlNoiseMult), B=Constant()), False=Multiply(A=@n88, B=@n90.CurlNoiseMult)) {parameterName=UseRVTHeightMap?}
  n92 = Add(
   A=Add(
    A=MaterialFunctionCall(WidthBase=@n85, WidthTip=@n87) {materialFunction=&SplineThicken}.WorldPositionOffset
    B=StaticSwitchParameter(True=@n91, False=@n88) {parameterName=UseInNiagara?}
   )
   B=VectorParameter() {parameterName=MoveOffGround, defaultValue=<rgba 0 0 0 1>}.RGB
  )
  n93 = ComponentMask(None=@n92) {b=true}
  n94 = NamedRerouteUsage()
  n95 = StaticSwitchParameter(True=AppendVector(A=ComponentMask(None=@n92) {g=true, r=true}, B=Add(A=@n93, B=@n94)), False=@n92) {parameterName=UseRVTHeightMap?}
  StaticSwitchParameter(True=@n95, False=@n95) {parameterName=UseInNiagara?}
 ]
}
